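# Animation Blueprint for the RaisinHunt Rag backpack cosmetic, adding secondary motion to its zippers and arrows.
# AnimGraph offsets the pack bone, then chains AnimDynamics spring bodies through cached poses: two zippers, six arrows, four simulating relative to their base bones.
# EventGraph computes backpack rotation and translation variables from spine bend relative to the ref pose.
# It also sets dynamics alpha and gravity variables for front end versus gameplay, lowering alpha during ziplines.
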AnimBlueprint /BRCosmetics/Accessories/FORT_Backpacks/Backpack_M_MED_RaisinHunt_Rag/Meshes/M_MED_RaisinHunt_Rag_Pack_AnimBP
target_skeleton=&/M_MED_RaisinHunt_Rag_Pack_Skeleton
graphs [
 AnimGraph {nodes [
   Output Pose(Result=Use cached pose 'AnimDynamicsCache_1_1'()) {layerGroup=None}
   n1 = "Transform (Modify) Bone - Bone: pack"!(ComponentPose=Local To Component(LocalPose=Input Pose()), Alpha, Translation="0,-1.5,0") -> (Pose) {
    rotationMode=BMM_Additive
    rotationSpace=BCS_BoneSpace
    translationMode=BMM_Additive
    alphaBoolBlend {blendOption=Linear}
    boneToModify {boneName=pack}
    translation {y=-1.5}
   }
   n2 = "Transform (Modify) Bone - Bone: pack"!(ComponentPose=@n1, Alpha, Translation="0,0,0", Rotation:Rotator="0,0,0") -> (Pose) {
    rotationMode=BMM_Additive
    rotationSpace=BCS_BoneSpace
    translationMode=BMM_Additive
    alphaBoolBlend {blendOption=Linear}
    boneToModify {boneName=pack}
   }
   n3 = "Transform (Modify) Bone - Bone: pack"!(ComponentPose=@n2, bAlphaBoolEnabled:bool, Translation="-5,-5,0", Rotation="0,0,10") -> (Pose) {
    alphaInputType=Bool
    rotationMode=BMM_Additive
    rotationSpace=BCS_BoneSpace
    translationMode=BMM_Additive
    alphaBoolBlend {blendInTime=0.5, blendOption=QuadraticInOut, blendOutTime=0.5}
    boneToModify {boneName=pack}
    rotation {roll=10}
    translation=<-5 -5>
   }
   Save cached pose 'ModifyCache_1_1'(Pose=Component To Local(ComponentPose=@n3)) {cacheName=ModifyCache_1_1}
   n4 = Anim Dynamics - Bone: dyn_Zipper_01!(ComponentPose=Local To Component(LocalPose=Use cached pose 'ModifyCache_1_1'()), Alpha=0.6, GravityOverride:Vector="0,0,-980") -> (Pose) {
    angularDampingOverride=1
    angularSpringConstant=9000
    bAngularSpring=true
    bOverrideAngularDamping=true
    bUseGravityOverride=true
    gravityScale=0.35
    simulationSpace=RootRelative
    boundBone {boneName=dyn_Zipper_01}
    componentAppliedLinearAccClamp=<3000 3000 3000>
    componentLinearAccScale=<0.4 0.4 0.4>
    componentLinearVelScale=<0.4 0.4 0.4>
    gravityOverride {z=-980}
    physicsBodyDefinitions [{
      boundBone {boneName=dyn_Zipper_01}
      boxExtents=<8 20 8>
      constraintSetup {angularLimitsMax=<0 20 0>, angularLimitsMin=<0 -70 -20>, angularTarget=<1 0 0>}
      localJointOffset=<6 0 0>
     }]
    simSpaceSettings {maxAngularAcceleration=20, maxAngularVelocity=10, simSpaceAngularAlpha=1}
   }
   n5 = Anim Dynamics - Bone: dyn_zipper_02!(ComponentPose=@n4, Alpha=0.6, GravityOverride:Vector="0,0,-980") -> (Pose) {
    angularDampingOverride=1
    angularSpringConstant=9000
    bAngularSpring=true
    bOverrideAngularDamping=true
    bUseGravityOverride=true
    gravityScale=0.35
    simulationSpace=RootRelative
    boundBone {boneName=dyn_zipper_02}
    componentAppliedLinearAccClamp=<3000 3000 3000>
    componentLinearAccScale=<0.4 0.4 0.4>
    componentLinearVelScale=<0.4 0.4 0.4>
    gravityOverride {z=-980}
    physicsBodyDefinitions [{
      boundBone {boneName=dyn_zipper_02}
      boxExtents=<8 20 8>
      constraintSetup {angularLimitsMax=<0 5 -10>, angularLimitsMin=<0 -70 -70>, angularTarget=<1 0 0>}
      localJointOffset=<6 0 0>
     }]
    simSpaceSettings {maxAngularAcceleration=20, maxAngularVelocity=10, simSpaceAngularAlpha=1}
   }
   Save cached pose 'RigidBodyCache_1_1'(Pose=Component To Local(ComponentPose=@n5)) {cacheName=RigidBodyCache_1_1}
   n6 = Anim Dynamics - Bone: dyn_arrow_a_01!(ComponentPose=Local To Component(LocalPose=Use cached pose 'RigidBodyCache_1_1'()), Alpha=0.7, GravityOverride:Vector="0,0,-980") -> (Pose) {
    angularDampingOverride=1
    angularSpringConstant=9000
    bAngularSpring=true
    bOverrideAngularDamping=true
    bUseGravityOverride=true
    gravityScale=0.35
    simulationSpace=RootRelative
    alphaScaleBias {scale=2.5}
    boundBone {boneName=dyn_arrow_a_01}
    componentAppliedLinearAccClamp=<3000 3000 3000>
    componentLinearAccScale=<0.4 0.4 0.4>
    componentLinearVelScale=<0.4 0.4 0.4>
    gravityOverride {z=-980}
    physicsBodyDefinitions [{
      boundBone {boneName=dyn_arrow_a_01}
      boxExtents=<30 20 20>
      constraintSetup {angularLimitsMax=<50 1.2 2>, angularLimitsMin=<-50 -0.2 0>, angularTarget=<1 0 0>}
      localJointOffset=<4 0 0>
     }]
    simSpaceSettings {maxAngularAcceleration=40, maxAngularVelocity=10, simSpaceAngularAlpha=0.7}
   }
   n7 = Anim Dynamics - Bone: dyn_arrow_b_01!(ComponentPose=@n6, Alpha=0.7, GravityOverride:Vector="0,0,-980") -> (Pose) {
    angularDampingOverride=1
    angularSpringConstant=9000
    bAngularSpring=true
    bOverrideAngularDamping=true
    bUseGravityOverride=true
    gravityScale=0.35
    simulationSpace=RootRelative
    boundBone {boneName=dyn_arrow_b_01}
    componentAppliedLinearAccClamp=<3000 3000 3000>
    componentLinearAccScale=<0.4 0.4 0.4>
    componentLinearVelScale=<0.4 0.4 0.4>
    gravityOverride {z=-980}
    physicsBodyDefinitions [{
      boundBone {boneName=dyn_arrow_b_01}
      boxExtents=<30 15 15>
      constraintSetup {angularLimitsMax=<50 3 0.7>, angularLimitsMin=<-50 -0.5 -2>, angularTarget=<1 0 0>}
      localJointOffset=<6 0 0>
     }]
    simSpaceSettings {maxAngularAcceleration=40, maxAngularVelocity=10, simSpaceAngularAlpha=0.7}
   }
   n8 = Anim Dynamics - Bone: dyn_arrow_c_01!(ComponentPose=@n7, Alpha=0.7, GravityOverride:Vector="0,0,-980") -> (Pose) {
    angularDampingOverride=0.999
    angularSpringConstant=20000
    bAngularSpring=true
    bOverrideAngularDamping=true
    bUseGravityOverride=true
    gravityScale=0.35
    simulationSpace=BoneRelative
    alphaBoolBlend {blendOption=Linear}
    boundBone {boneName=dyn_arrow_c_01}
    componentAppliedLinearAccClamp=<3000 3000 3000>
    componentLinearAccScale=<0.5 0.5 0.5>
    componentLinearVelScale=<0.5 0.5 0.5>
    gravityOverride {z=-980}
    physicsBodyDefinitions [{
      boundBone {boneName=dyn_arrow_c_01}
      boxExtents=<40 40 20>
      constraintSetup {angularLimitsMax=<50 2.5 0.5>, angularLimitsMin=<-50 -0.2 -2>, angularTarget=<1 0 0>}
      localJointOffset=<2 0 0>
     }]
    relativeSpaceBone {boneName=base_arrow_c_01}
   }
   n9 = Anim Dynamics - Bone: dyn_arrow_d_01!(ComponentPose=@n8, Alpha=0.7, GravityOverride:Vector="0,0,-980") -> (Pose) {
    angularDampingOverride=0.999
    angularSpringConstant=10000
    bAngularSpring=true
    bOverrideAngularDamping=true
    bUseGravityOverride=true
    gravityScale=0.35
    simulationSpace=BoneRelative
    alphaBoolBlend {blendOption=Linear}
    boundBone {boneName=dyn_arrow_d_01}
    componentAppliedLinearAccClamp=<3000 3000 3000>
    componentLinearAccScale=<0.5 0.5 0.5>
    componentLinearVelScale=<0.5 0.5 0.5>
    gravityOverride {z=-980}
    physicsBodyDefinitions [{
      boundBone {boneName=dyn_arrow_d_01}
      boxExtents=<40 40 20>
      constraintSetup {angularLimitsMax=<0 1 0.5>, angularLimitsMin=<0 -2 1>, angularTarget=<1 0 0>}
      localJointOffset=<2 0 0>
     }]
    relativeSpaceBone {boneName=base_arrow_d_01}
   }
   n10 = Anim Dynamics - Bone: dyn_arrow_e_01!(ComponentPose=@n9, Alpha=0.7, GravityOverride:Vector="0,0,-1500") -> (Pose) {
    angularDampingOverride=0.999
    angularSpringConstant=10000
    bAngularSpring=true
    bOverrideAngularDamping=true
    bUseGravityOverride=true
    gravityScale=0.35
    simulationSpace=BoneRelative
    alphaBoolBlend {blendOption=Linear}
    boundBone {boneName=dyn_arrow_e_01}
    componentAppliedLinearAccClamp=<3000 3000 3000>
    componentLinearAccScale=<0.5 0.5 0.5>
    componentLinearVelScale=<0.5 0.5 0.5>
    gravityOverride {z=-1500}
    physicsBodyDefinitions [{
      boundBone {boneName=dyn_arrow_e_01}
      boxExtents=<40 40 20>
      constraintSetup {angularLimitsMax=<50 2 2.2>, angularLimitsMin=<-50 -0.8 -2>, angularTarget=<1 0 0>}
      localJointOffset=<4 0 0>
     }]
    relativeSpaceBone {boneName=base_arrow_e_01}
    simSpaceSettings {simSpaceAngularAlpha=0.2}
   }
   n11 = Anim Dynamics - Bone: dyn_arrow_f_01!(ComponentPose=@n10, Alpha=0.7, GravityOverride:Vector="0,0,-980") -> (Pose) {
    angularDampingOverride=0.998
    angularSpringConstant=10000
    bAngularSpring=true
    bOverrideAngularDamping=true
    bUseGravityOverride=true
    gravityScale=0.35
    simulationSpace=BoneRelative
    alphaBoolBlend {blendOption=Linear}
    boundBone {boneName=dyn_arrow_f_01}
    componentAppliedLinearAccClamp=<3000 3000 3000>
    componentLinearAccScale=<0.5 0.5 0.5>
    componentLinearVelScale=<0.5 0.5 0.5>
    gravityOverride {z=-980}
    physicsBodyDefinitions [{
      boundBone {boneName=dyn_arrow_f_01}
      boxExtents=<40 40 20>
      constraintSetup {angularLimitsMax=<50 1.5 1.2>, angularLimitsMin=<-50 0 -2>, angularTarget=<1 0 0>}
      localJointOffset=<6 0 0>
     }]
    relativeSpaceBone {boneName=base_arrow_f_01}
    simSpaceSettings {maxAngularAcceleration=80, maxAngularVelocity=16, simSpaceAngularAlpha=0.35}
   }
   Save cached pose 'AnimDynamicsCache_1_1'(Pose=Component To Local(ComponentPose=@n11)) {cacheName=AnimDynamicsCache_1_1}
  ]}
 EventGraph {nodes [
   Event BlueprintUpdateAnimation()
   n1 = GetOwnerSkeletalMeshComponent()
   n2 = BreakRotator(InRot=BreakTransform(InTransform=GetDeltaTransformFromRefPose(self=@n1, BoneName=spine_01, BaseName=spine_03)).Rotation)
   > Set backpackRotate(backpackRotate:Rotator=MakeRotator(Roll=MapRangeClamped(Value=@n2.Yaw, InRangeA=0, InRangeB=25, OutRangeB=30))) -> (Output_Get:Rotator)
   n3 = MapRangeClamped(Value=@n2.Yaw, InRangeA=0, InRangeB=35, OutRangeB=-5)
   > Set BackpackTranslate(BackpackTranslate:Vector=MakeVector(X=@n3, Y=@n3)) -> (Output_Get:Vector)
   Event OnCharacterCustomizationLoaded()
   > n4 = Branch(Condition=Get bIsInFrontEnd())
   > Set DynamicsAlpha(DynamicsAlpha:double=0.6) -> (Output_Get:double)
   > Set Override World Gravity(Override World Gravity:Vector="0,0,-580") -> (Output_Get:Vector)
   Set DynamicsAlpha(execute=@n4.else, DynamicsAlpha=0.9)
   > Set Override World Gravity(Override World Gravity="0,0,-850")
   n5 = Event BlueprintProcessAnimInputEvent()
   > n6 = Branch(Condition=IsNamed(self=@n5.AnimInputEvent, InEventName=aie.Player.Zipline.Entered))
   > Set DynamicsAlpha(DynamicsAlpha=0.25)
   Branch(execute=@n6.else, Condition=IsNamed(self=@n5.AnimInputEvent, InEventName=aie.Player.Zipline.Exited))
   > Set DynamicsAlpha(DynamicsAlpha=0.9)
  ]}
]
variables [
 backpackRotate {type=Rotator}
 BackpackTranslate {type=Vector}
 DynamicsAlpha {type=number}
 Override World Gravity {type=Vector}
]
